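# Material function that stamps a rank icon onto the RankedDefender backpack's material attributes.
# RankIndex selects a cell in a shared 3x3 icon mask and normal atlas through offset, scaled UVs.
# A box mask from pre-skinned position confines the icon, with a static switch for debugging it.
# The icon blends in biased base color, roughness and metallic, an emissive color, and an angle-corrected normal.
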MaterialFunction /BRCosmetics/Accessories/FORT_Backpacks/Backpack_M_MED_RankedDefender/Material/MF_RankedDefenderIcon
graph {nodes [
  ChannelMaskParameter(None=FunctionInput() {bUsePreviewValueAsDefault=true, inputName=Alpha}) {group=RankIcon, maskChannel=Blue, parameterName=RH_IconMaskChannel, defaultValue=<rgba 0 0 1 0>}
  n1 = GetMaterialAttributes(None=FunctionInput() {inputName=Material, inputType=FunctionInput_MaterialAttributes}) {attributeGetTypes [
    #1
    #2
    #3
    #4
   ]}
  n2 = NamedRerouteUsage()
  n3 = ScalarParameter() {defaultValue=-0.3, group=RankIcon, parameterName=BaseColorBias}
  n4 = ScalarParameter() {defaultValue=-0.1, group=RankIcon, parameterName=RoughnessBias}
  n5 = ScalarParameter() {group=RankIcon, parameterName=MetalBias}
  n6 = SetMaterialAttributes!(
   MaterialAttributes=@n1.MaterialAttributes
   BaseColor=Saturate(None=Add(A=@n1.BaseColor, B=@n3))
   Roughness=Saturate(None=Add(A=@n1.Roughness, B=@n4))
   Metallic=Saturate(None=Add(A=@n1.Metallic, B=@n5))
   Normal=NamedRerouteUsage()
   EmissiveColor=VectorParameter() {group=RankIcon, parameterName=IconEmissiveColor, defaultValue=<rgba 0 0 0 1>}.RGB
  ) {attributeSetTypes [
    #1
    #2
    #3
    #4
    #5
   ]}
  n7 = NamedRerouteUsage()
  FunctionOutput(None=StaticSwitchParameter(
    True=SetMaterialAttributes!(MaterialAttributes=@n1.MaterialAttributes, BaseColor=@n2) {attributeSetTypes [#1]}
    False=BlendMaterialAttributes(A=@n1.MaterialAttributes, B=@n6, Alpha=@n7) {vertexAttributeBlendType=UseA}
   ) {group=RankIcon, parameterName=DEBUG-IconBoxMask})
  n8 = NamedRerouteUsage()
  n9 = NamedRerouteUsage()
  NamedRerouteDeclaration(None=Multiply(A=TextureSample(UVs=@n8) {samplerType=SAMPLERTYPE_Masks, texture=/Game/Global/Materials/Cosmetics/Ranked/Textures/T_Ranked_Icon3x3}.R, B=@n9)) {name=IconMask, nodeColor=<rgba 0.5469 0 1 1>}
  NamedRerouteDeclaration(None=@n1.Normal) {name=IN_Normal, nodeColor=<rgba 1 0.4688 0 1>}
  n10 = Constant3Vector() {constant=<rgba 0 0 1 0>}
  n11 = MaterialFunctionCall(
   Normal=TextureSample(UVs=NamedRerouteUsage()) {samplerType=SAMPLERTYPE_Normal, texture=/Game/Global/Materials/Cosmetics/Ranked/Textures/T_Ranked_Icon3x3_N}.RGB
   Flatness=ScalarParameter() {group=RankIcon, parameterName=IconNormalFlatness}
  ) {materialFunction=&FlattenNormal}
  NamedRerouteDeclaration(None=MaterialFunctionCall(BaseNormal=NamedRerouteUsage(), AdditionalNormal=LinearInterpolate(A=@n10, B=@n11.Result, Alpha=@n9)) {materialFunction=&BlendAngleCorrectedNormals}.Result) {name=OUT_Normal, nodeColor=<rgba 0.125 1 0 1>}
  n12 = Divide(A=VertexInterpolator(VS=PreSkinnedPosition()).PS) {constB=100}
  n13 = Step(Y=ScalarParameter() {group=RankIcon, parameterName=SphereMaskCutOff}, X=ComponentMask(None=@n12) {g=true})
  n14 = MaterialFunctionCall!(
   A=ComponentMask(None=@n12) {b=true, r=true}
   B=AppendVector(
    A=ScalarParameter() {group=RankIcon, parameterName=SphereMaskX}
    B=ScalarParameter() {defaultValue=-0.2, group=RankIcon, parameterName=SphereMaskY}
   )
   Bounds=AppendVector(
    A=ScalarParameter() {defaultValue=0.045, group=RankIcon, parameterName=SphereMaskBoundsX}
    B=ScalarParameter() {defaultValue=0.02, group=RankIcon, parameterName=SphereMaskBoundsY}
   )
   Edge Falloff=Constant() {r=0.01}
  ) {materialFunction=/Engine/Functions/Engine_MaterialFunctions01/Gradient/BoxMask-2D}
  NamedRerouteDeclaration(None=Multiply(A=OneMinus(None=@n13), B=Step(X=@n14.Result) {constY=0.1})) {name=IconBoxMask, nodeColor=<rgba 0.9219 1 0 1>}
  n15 = ScalarParameter() {group=RankIcon, parameterName=RankIndex}
  n16 = Constant() {r=3}
  n17 = Fmod(A=@n15, B=@n16)
  n18 = Constant() {r=3}
  n19 = Divide(B=@n18)
  n20 = Floor(None=Divide(A=@n15, B=@n18))
  NamedRerouteDeclaration(None=MaterialFunctionCall(X=Divide(A=@n17, B=@n16), Y=Multiply(A=@n19, B=@n20)) {materialFunction=&MakeFloat2}.Result) {name=RankOffset, nodeColor=<rgba 0.0312 0 1 1>}
  n21 = MaterialFunctionCall(UVs=TextureCoordinate(), Texture Scale=ScalarParameter() {defaultValue=0.4, group=RankIcon, parameterName=IconUVScale}) {materialFunction=&ScaleUVsByCenter}
  n22 = MaterialFunctionCall(
   X=ScalarParameter() {defaultValue=-0.664, group=RankIcon, parameterName=IconUVOffsetX}
   Y=ScalarParameter() {defaultValue=0.67, group=RankIcon, parameterName=IconUVOffsetY}
  ) {materialFunction=&MakeFloat2}
  NamedRerouteDeclaration(None=Add(A=NamedRerouteUsage(), B=Add(A=@n21.UVs, B=@n22.Result))) {name=UVs, nodeColor=<rgba 0.4297 0 1 1>}
 ]}
